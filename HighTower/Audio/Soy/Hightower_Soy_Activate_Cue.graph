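# Sound cue for the activation of the HighTower Soy feature, playing the Hightower_Soy_Leap wave.
# A single wave player feeds a root modulator that slightly randomizes volume on each play.
# It uses a close-range weapon projectile attenuation asset and is routed to a first-person ranged weapon fire sound class.
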SoundCue /HighTower/Audio/Soy/Hightower_Soy_Activate_Cue
root=Modulator_0
nodes [
 {id=Modulator_0, children [WavePlayer_0], properties {volumeMax=1, volumeMin=0.9}}
 {id=WavePlayer_0, properties {soundWaveAssetPtr=&/Hightower_Soy_Leap}}
]
properties {
 attenuationSettings=&Fort_Weapon_Projectile_Close_Athena_Att
 maxDistance=4950
 soundClassObject=&Fort_WeaponRanged_Fire_1P_PSM
}
