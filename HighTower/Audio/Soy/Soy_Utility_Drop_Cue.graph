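# Sound cue for dropping the HighTower Soy utility item, built from Soy equip sound waves.
# A mixer blends a louder layer, a random pick of two waves pitched up with slight volume variation, with a quieter single wave pitched down.
# It plays under the loot collect sound class with a close-range weapon attenuation asset.
# A concurrency override allows one instance at a time and blocks new plays while it runs.
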SoundCue /HighTower/Audio/Soy/Soy_Utility_Drop_Cue
root=Mixer_0
nodes [
 {id=Mixer_0, children [Modulator_0,Modulator_1], properties {inputVolume [0.7,0.3]}}
 {id=Modulator_0, children [Random_0], properties {pitchMax=1.3, pitchMin=1.2, volumeMax=1.2, volumeMin=1.1}}
 {id=Modulator_1, children [WavePlayer_3], properties {pitchMax=0.85, pitchMin=0.8}}
 {id=Random_0, children [WavePlayer_1,WavePlayer_2], properties {weights [1,1]}}
 {id=WavePlayer_1, properties {soundWaveAssetPtr=&/Hightower_Soy_Equip_12}}
 {id=WavePlayer_2, properties {soundWaveAssetPtr=&/Hightower_Soy_Equip_11}}
 {id=WavePlayer_3, properties {soundWaveAssetPtr=&/Hightower_Soy_Equip_02}}
]
properties {
 attenuationSettings=/Game/Sounds/Attenuation/Fort_Weapon_Umbrella_Close_Athena_Att
 bOverrideConcurrency=true
 maxDistance=4500
 pitchMultiplier=0.9
 soundClassObject=&Fort_Loot_Collect
 volumeMultiplier=0.55
 concurrencyOverrides {maxCount=1, resolutionRule=PreventNew}
}
