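# Sound cue for picking up a Soy utility item, routed through the loot-collect sound class.
# A random node picks evenly between two Soy equip waves, then a modulator pitches the result well up with slight pitch and volume variation.
# It uses a close weapon attenuation asset and overrides concurrency to allow one instance, blocking new plays while one is active.
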SoundCue /HighTower/Audio/Soy/Soy_Utility_Pickup_Cue
root=Modulator_0
nodes [
 {id=Modulator_0, children [Random_0], properties {pitchMax=1.6, pitchMin=1.5, volumeMax=0.75, volumeMin=0.7}}
 {id=Random_0, children [WavePlayer_1,WavePlayer_2], properties {weights [1,1]}}
 {id=WavePlayer_1, properties {soundWaveAssetPtr=&/Hightower_Soy_Equip_12}}
 {id=WavePlayer_2, properties {soundWaveAssetPtr=&/Hightower_Soy_Equip_11}}
]
properties {
 attenuationSettings=/Game/Sounds/Attenuation/Fort_Weapon_Umbrella_Close_Athena_Att
 bOverrideConcurrency=true
 maxDistance=4500
 pitchMultiplier=0.9
 soundClassObject=&Fort_Loot_Collect
 volumeMultiplier=0.5
 concurrencyOverrides {maxCount=1, resolutionRule=PreventNew}
}
